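; Two-sided quest scoreboard that tracks scores for sides A and B on two sign meshes sharing per-side dynamic materials.
; MoveScore records contributing players and increments replicated scores after a delay; OnReps update client materials and play sounds.
; EndGame grants quest progress to all contributors outside Creative, then triggers the winner's effects, with ties going to B.
; Clients bind quest updates after playlist load and a five-second workaround delay, playing success FX on completion.
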(event EventBeginPlay
  (bind _returnvalue (Fortnite|GetLocalFortPlayerControllers true))
  (bind _score_b_index (Variables|SignContent|GetScore_B_Index))
  (bind _score_a_index (Variables|SignContent|GetScore_A_Index))
  (bind _output (Utilities|Array|Get(acopy) _returnvalue))
  ;[0 0.9646 1] Initialize
  (Networking|SwitchHasAuthority
    (:Remote
      ;; Base Score Setup
      (Utilities|FlowControl|Sequence
        (:then_0
          (bind _returnvalue_2 (Rendering|Material|CreateDynamicMaterialInstance (Variables|Default|GetScoreSign_A) _score_a_index 0 "None" :!class /Script/Engine.PrimitiveComponent))
          (bind _output_get (Variables|ScoringInfo|SetMID_Score_A _returnvalue_2))
          (Rendering|Material|SetMaterial (Variables|Default|GetScoreSign_B) _score_a_index _output_get))
        (:then_1
          (bind _returnvalue_3 (Rendering|Material|CreateDynamicMaterialInstance (Variables|Default|GetScoreSign_B) _score_b_index 0 "None" :!class /Script/Engine.PrimitiveComponent))
          (bind _output_get_2 (Variables|ScoringInfo|SetMID_Score_B _returnvalue_3))
          (Rendering|Material|SetMaterial (Variables|Default|GetScoreSign_A) _score_b_index _output_get_2))
        (:then_2
          (Rendering|Material|SetScalarParameterValue (Variables|ScoringInfo|GetMID_Score_A) "Score" 0.0 :!class /Script/Engine.MaterialInstanceDynamic)
          (Rendering|Material|SetVectorParameterValue (Variables|ScoringInfo|GetMID_Score_A) "Color" (Variables|SignContent|GetScoreColor_A) :!class /Script/Engine.MaterialInstanceDynamic)
          (Rendering|Material|SetVectorParameterValue (Variables|ScoringInfo|GetMID_Score_B) "Color" (Variables|SignContent|GetScoreColor_B) :!class /Script/Engine.MaterialInstanceDynamic))
        (:then_3
          _))
      ;[1 1 1] Check if In Creative Mode or not
      (bind _asfort_game_state_athena (Utilities|Casting|CastToFortGameStateAthena (Game|GetGameState)))
      (if (Variables|/Script/FortniteGame.FortGameStateAthena|GetbPlaylistDataIsLoaded _asfort_game_state_athena)
        (if (GameplayTags|HasTag (Class|FortGameStateAthena|GetAthenaPlaylistContextTags _asfort_game_state_athena) "(TagName=\"Athena.Creative\")")
          (Variables|Default|SetCreativeMode true)
          (else
            (Variables|Default|SetCreativeMode false)))
        ;[1 0.5839 0] HackyHack Pls Fix Lietz
        (Utilities|FlowControl|Delay 5.0)
        (Networking|SwitchHasAuthority
          (:Remote
            ;[1 1 1] Quest Bindings
            (bind (_questisvalid _questcompleted _objectivecompleted) (Default|AthenaCheckQuestandObjectiveCompletion _output (Variables|Default|GetQuestItem) (Variables|Default|GetObjBackendName)))
            (if _questisvalid
              (if _objectivecompleted
                _
                (elif (Variables|Default|GetCreativeMode)
                  _
                  (else
                    (Quests|BindEventtoOnQuestsUpdated (Quests|GetQuestManager _output "Invalid") (EventDispatchers|CreateEvent :!event Custom|QuestUpdated))
                    (Variables|Default|SetLocallyActive true)
                    (Development|PrintString "QUEST BOUND" false false "(R=0.000000,G=0.660000,B=1.000000,A=1.000000)" 1e+01)))))))
        (else
          (Playlist|BindEventtoOnCurrentPlaylistLoaded _asfort_game_state_athena (delegate Custom|PlaylistLoaded)))))))

(event Custom|PlaylistLoaded
  (params (PlaylistName :type Name) (PlaylistContextTags :type Gameplay%20Tag%20Container%20Structure%20%28by%20ref%29))
  (bind _returnvalue (Fortnite|GetLocalFortPlayerControllers true))
  (bind _output (Utilities|Array|Get(acopy) _returnvalue))
  ;[1 1 1] Check if In Creative Mode or not
  (if (GameplayTags|HasTag PlaylistContextTags "(TagName=\"Athena.Creative\")")
    (Variables|Default|SetCreativeMode true)
    (else
      (Variables|Default|SetCreativeMode false)))
  ;[1 0.5839 0] HackyHack Pls Fix Lietz
  (Utilities|FlowControl|Delay 5.0)
  (Networking|SwitchHasAuthority
    (:Remote
      ;[1 1 1] Quest Bindings
      (bind (_questisvalid _questcompleted _objectivecompleted) (Default|AthenaCheckQuestandObjectiveCompletion _output (Variables|Default|GetQuestItem) (Variables|Default|GetObjBackendName)))
      (if _questisvalid
        (if _objectivecompleted
          _
          (elif (Variables|Default|GetCreativeMode)
            _
            (else
              (Quests|BindEventtoOnQuestsUpdated (Quests|GetQuestManager _output "Invalid") (EventDispatchers|CreateEvent :!event Custom|QuestUpdated))
              (Variables|Default|SetLocallyActive true)
              (Development|PrintString "QUEST BOUND" false false "(R=0.000000,G=0.660000,B=1.000000,A=1.000000)" 1e+01))))))))

(event Custom|QuestUpdated
  ;[1 1 1] Quest Updated
  (if (Variables|Default|GetLocallyActive)
    (bind (_questisvalid _questcompleted _objectivecompleted) (Default|AthenaCheckQuestandObjectiveCompletion (Utilities|Array|Get(acopy) (Fortnite|GetLocalFortPlayerControllers true)) (Variables|Default|GetQuestItem) (Variables|Default|GetObjBackendName)))
    (if (or _questcompleted _objectivecompleted)
      (Variables|Default|SetLocallyActive false)
      (Quests|UnbindEventfromOnQuestsUpdated (Quests|GetQuestManager (Utilities|Array|Get(acopy) (Fortnite|GetLocalFortPlayerControllers true)) "Invalid") (EventDispatchers|CreateEvent :!event Custom|QuestUpdated))
      (Utilities|FlowControl|Delay 1.5)
      (CallFunction|PlayQuestSuccessFX))))

(event EventStartup
  ;[1 1 1] Initiate
  (Networking|SwitchHasAuthority
    (:Authority
      (Variables|ScoringInfo|SetScore_A 0)
      (Variables|ScoringInfo|SetScore_B 0))))

(event Custom|PlayQuestSuccessFX
  (bind _returnvalue (Fortnite|GetLocalFortPlayerControllers true))
  (bind _output (Utilities|Array|Get(acopy) _returnvalue))
  (bind _returnvalue_2 (Game|Player|GetPlayerPawn _output))
  ;[1 1 1] Quest Success FX
  (Development|PrintString "SUCCESS FX" false false "(R=1.000000,G=0.703857,B=0.040648,A=1.000000)" 5.0)
  (Effects|SpawnFXSystemAttached (Variables|Default|GetQuestSuccessEmitter) (Variables|/Script/Engine.Character|GetMesh _returnvalue_2))
  (Audio|PlaySoundatLocation (Variables|Default|GetQuestSuccessSound) (Transformation|GetActorLocation _returnvalue_2)) ; Quest Success Sound
)

(event EventEndGame
  (bind _allscoreplayers (Variables|Default|GetAllScorePlayers))
  (bind _allscoreplayers_2 (Variables|Default|GetAllScorePlayers))
  ;[0.3333 0.3333 0.3333] Win Condition and End Game
  ;;[0.2601 0.3249 1] Current Score exceeds High
  (Utilities|FlowControl|Sequence
    (:then_0
      ;[1 1 1] Check for Quest Update
      (if (not (Variables|Default|GetCreativeMode))
        ;;[1 1 1] Combine scores for tie or allplayers
        (Utilities|Array|AppendArray _allscoreplayers (Variables|Default|GetScoringPlayers_A))
        (Utilities|Array|AppendArray _allscoreplayers (Variables|Default|GetScoringPlayers_B))
        (Class|AthenaQuestFunctionLibrary|AthenaGiveQuestUpdatetoPlayers (Variables|Default|GetQuestItem) (Variables|Default|GetObjBackendName) (Variables|Default|GetObjStat) _allscoreplayers_2)
        (Development|PrintString (Utilities|String|ToString(Integer) (Utilities|Array|Length _allscoreplayers_2)) false false "(R=1.000000,G=0.000000,B=0.781195,A=1.000000)" 1e+01)))
    (:then_1
      (if (> (Variables|ScoringInfo|GetScore_A) (Variables|ScoringInfo|GetScore_B))
        ;;[0.3906 0.5729 0.1969] Winning Team
        (CallFunction|SideAWon)
        (else
          (CallFunction|SideBWon))))))

(event Custom|Side_A_Won
  (bind _scoresign_a (Variables|Default|GetScoreSign_A))
  (bind _returnvalue (Transformation|GetWorldLocation _scoresign_a))
  ;[1 0.9976 0] Win Behavior on board / Client Only
  (Networking|SwitchHasAuthority
    (:Remote
      (if (not (Variables|Default|GetSuppressScoring))
        (Audio|PlaySoundatLocation (Variables|SignContent|GetNewHighScoreSound) _returnvalue) ; High Score Sound
        (Utilities|FlowControl|Delay (Variables|ScoringInfo|GetEnd%20Game%20Delay))
        (Components|Activation|Activate (Variables|Default|GetWinFX_A) false)
        (Audio|PlaySoundatLocation (Variables|SignContent|GetFireworksSound) _returnvalue) ; High Score Sound
      ))))

(event Custom|Side_B_Won
  (bind _scoresign_b (Variables|Default|GetScoreSign_B))
  (bind _returnvalue (Transformation|GetWorldLocation _scoresign_b))
  ;[0.2917 0.2917 0.2917] Score Behavior
  ;[1 0.08653 0.7576] AUDIO - Play Side B Won
  (Networking|SwitchHasAuthority
    (:Remote
      (if (not (Variables|Default|GetSuppressScoring))
        (Audio|PlaySoundatLocation (Variables|SignContent|GetNewHighScoreSound) _returnvalue "0, 0, 0" 2.0) ; Amplified to be audible
        (Utilities|FlowControl|Delay (Variables|ScoringInfo|GetEnd%20Game%20Delay))
        (Components|Activation|Activate (Variables|Default|GetWinFX_B) false)
        (Audio|PlaySoundatLocation (Variables|SignContent|GetFireworksSound) _returnvalue) ; High Score Sound
      ))))

(event EventMoveScore
  (params (ScoreIncrement :type Integer) (bIsSideA :type Boolean) (ContributingPlayers_A :type Array%20of%20Fort%20Player%20Controller%20Athena%20Object%20References) (ContributingPlayers_B :type Array%20of%20Fort%20Player%20Controller%20Athena%20Object%20References))
  ;[0.2917 0.2917 0.2917] Score Behavior
  ;;[0.3217 0.8799 1] External Event to move Score up (or down)
  (Utilities|FlowControl|Sequence
    (:then_0
      (for _array_element ContributingPlayers_A
        (Utilities|Array|AddUnique (Variables|Default|GetScoringPlayers_A) _array_element)))
    (:then_1
      (for _array_element_2 ContributingPlayers_B
        (Utilities|Array|AddUnique (Variables|Default|GetScoringPlayers_B) _array_element_2)))
    (:then_2
      (Networking|ForceNetUpdate)
      (if bIsSideA
        (Utilities|FlowControl|Delay (Variables|ScoringInfo|GetScoreDelay))
        (Variables|ScoringInfo|SetScore_A (+ (Variables|ScoringInfo|GetScore_A) ScoreIncrement))
        (else
          (Utilities|FlowControl|Delay (Variables|ScoringInfo|GetScoreDelay))
          (Variables|ScoringInfo|SetScore_B (+ (Variables|ScoringInfo|GetScore_B) ScoreIncrement)))))))

(fn OnRep_Score_A
  (bind _score_a (Variables|ScoringInfo|GetScore_A))
  (Networking|SwitchHasAuthority
    (:Remote
      (if (not (== _score_a 0))
        (Audio|PlaySoundatLocation (Variables|SignContent|GetScore%20Up%20Sound) (Transformation|GetWorldLocation (Variables|Default|GetScoreSign_A))) ; Score Up Sound
      )
      (Rendering|Material|SetScalarParameterValue (Variables|ScoringInfo|GetMID_Score_A) "Score" (Math|Conversions|ToFloat(Integer) _score_a) :!class /Script/Engine.MaterialInstanceDynamic))))

(fn OnRep_Score_B
  (bind _score_b (Variables|ScoringInfo|GetScore_B))
  (Networking|SwitchHasAuthority
    (:Remote
      (if (not (== _score_b 0))
        (Audio|PlaySoundatLocation (Variables|SignContent|GetScore%20Up%20Sound) (Transformation|GetWorldLocation (Variables|Default|GetScoreSign_B))) ; Score Up Sound
      )
      (Rendering|Material|SetScalarParameterValue (Variables|ScoringInfo|GetMID_Score_B) "Score" (Math|Conversions|ToFloat(Integer) _score_b) :!class /Script/Engine.MaterialInstanceDynamic))))

(fn ConstructionScript
  (bind _sign_mesh (Variables|SignContent|GetSign%20Mesh))
  ; Mesh Setup
  (if (Utilities|NotEqual(Object) _sign_mesh 0)
    (Components|StaticMesh|SetStaticMesh (Variables|Default|GetScoreSign_B) _sign_mesh)
    (Components|StaticMesh|SetStaticMesh (Variables|Default|GetScoreSign_A) _sign_mesh)
    (else
      _))
  ; Win Particle Setup
  (Effects|Components|ParticleSystem|SetTemplate (Variables|Default|GetWinFX_A) (Variables|SignContent|GetWinEffect)))
