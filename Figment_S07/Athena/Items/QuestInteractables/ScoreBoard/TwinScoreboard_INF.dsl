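; Blueprint interface for the twin scoreboard quest interactable in Figment S07.
; It declares Startup, EndGame and MoveScore without implementations, leaving scoring logic to implementing classes.
; MoveScore passes an integer increment, a flag for whether side A scores, and separate arrays of contributing Athena player controllers for sides A and B.
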(fn EndGame
  (return))

(fn MoveScore
  (params (ScoreIncrement :type Integer) (bIsSideA :type Boolean) (ContributingPlayers_A :type Array%20of%20Fort%20Player%20Controller%20Athena%20Object%20References) (ContributingPlayers_B :type Array%20of%20Fort%20Player%20Controller%20Athena%20Object%20References))
  (return))

(fn Startup)
